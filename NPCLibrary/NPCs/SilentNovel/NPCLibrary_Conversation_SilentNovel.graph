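# Conversation database for the SilentNovel NPC, defining its greeting speech and the menu of dialogue branches that follow it.
# One speech node picks contextual lines by priority for first-ever meetings, match conversation count, or a specific player outfit.
# It links out by entry tag to shared quest and basic-service conversations, plus Season 38 story, NPC-specific and vehicle branches.
# The vehicle link requires that no LAF quest is active.
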ConversationDatabase /NPCLibrary/NPCs/SilentNovel/NPCLibrary_Conversation_SilentNovel
entry_points [{tag=Conversation.Entry.SilentNovel, destinations [#1]}]
nodes [
 ConversationEntryPointNode_7 {
  id=#1
  class=ConversationEntryPointNode
  outputs [#2]
  properties {entryTag {tagName=Conversation.Entry.SilentNovel}}
 }
 Speech_C_0 {
  id=#2
  class=Speech_C
  outputs [
   #3
   #4
   #5
   #6
   #7
  ]
  properties {
   nodeName=""
   generalConfig {
    defaultMessage="Can't talk, busy making everything better."
    contextualMessages [
     {
      message="No need to thank me, you're already welcome."
      contextRequirements [{requirement {class=FortControllerRequirement_IsFirstEverConversationWithNPC}}]
      priority {curve {rowName=Tandem.Speech.Priority.FirstEverInteraction}}
     }
     {
      message="Can't talk, busy making everything better."
      contextRequirements [{requirement {class=FortControllerRequirement_CheckNPCMatchConversationCount, properties {nPCUniqueIdTag {tagName=AISpawnerData.Type.Tandem.SilentNovel}}}}]
      priority {curve {rowName=None}}
     }
     {
      message=Stupid Flanders.
      contextRequirements [{requirement {class=FortControllerRequirement_HasCID, properties {allowedCIds [/BRCosmetics/Athena/Items/Cosmetics/Characters/Character_CoolSlice_Golf]}}}]
      priority {value=2, curve {rowName=Tandem.Speech.Priority.OutfitReaction}}
     }
    ]
   }
  }
 }
 ConversationLinkNode_4 {
  id=#7
  class=ConversationLinkNode
  properties {remoteEntryTag {tagName=Conversation.Entry.Season38.LAF.NPC.Vehicle}}
  sub_nodes [FortConversationRequirement_ControllerRequirement_2 {
    class=FortConversationRequirement_ControllerRequirement
    properties {
     nodeName="Require: Has not started a L&&F Quest"
     requirement {class=FortControllerRequirement_HasOwnedTag, properties {bInverseCondition=true, ownedTag {tagName=Athena.Quests.S38.LAF.QuestActive}}}
    }
   }]
 }
 ConversationLinkNode_3 {
  id=#6
  class=ConversationLinkNode
  properties {remoteEntryTag {tagName=Conversation.Entry.Season38.SilentNovel}}
 }
 ConversationLinkNode_2 {
  id=#5
  class=ConversationLinkNode
  properties {remoteEntryTag {tagName=Conversation.Entry.Season38.Story.P02.Q07}}
 }
 ConversationLinkNode_0 {id=#4, class=ConversationLinkNode, properties {remoteEntryTag {tagName=Conversation.Entry.BasicServices}}}
 ConversationLinkNode_1 {id=#3, class=ConversationLinkNode, properties {remoteEntryTag {tagName=Conversation.Entry.NPCLibrary.Quests}}}
]
